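# Sound cue that plays the blast sound for the Orb Fragment gadget.
# A single wave, Charge_Fragment_Collect_Blasts_01, feeds a Modulator at default settings for slight random pitch and volume variation.
# The cue uses the gadget orb-activate attenuation and gadget sound class at reduced volume.
# Its concurrency override prevents new instances from starting once the limit is reached, rather than cutting off playing ones.
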SoundCue /Game/Sounds/Fort_Gadgets/Orb_Fragment/Orb_Fragment_Blast_Cue
root=Modulator_0
nodes [
 {id=Modulator_0, children [WavePlayer_0]}
 {id=WavePlayer_0, properties {soundWaveAssetPtr=&/Charge_Fragment_Collect_Blasts_01}}
]
properties {
 attenuationSettings=/Game/Sounds/Attenuation/Fort_Gadget_Orb_Activate_Att
 bOverrideConcurrency=true
 maxDistance=4000
 soundClassObject=/Game/Sounds/FortSoundClassesAndModes/Fort_Gadgets_StW
 volumeMultiplier=0.65
 concurrencyOverrides {resolutionRule=PreventNew}
}
