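# Sound cue for the Orb Fragment gadget's goodbye sound, played at reduced volume through the gadget sound class.
# A fixed one-second delay feeds a random node that picks among four equally weighted goodbye wave variants.
# It borrows the shared gadget-activate attenuation and overrides concurrency so new instances are blocked once the limit is reached.
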SoundCue /Game/Sounds/Fort_Gadgets/Orb_Fragment/Orb_Fragment_Goodbye_Cue
root=Delay_3
nodes [
 {id=Delay_3, children [Random_5], properties {delayMax=1, delayMin=1}}
 {id=Random_5, children [WavePlayer_25, WavePlayer_26, WavePlayer_27, WavePlayer_28], properties {weights [1,1,1,1]}}
 {id=WavePlayer_25, properties {soundWaveAssetPtr=&/orb_fragment_goodbye_01}}
 {id=WavePlayer_26, properties {soundWaveAssetPtr=&/orb_fragment_goodbye_02}}
 {id=WavePlayer_27, properties {soundWaveAssetPtr=&/orb_fragment_goodbye_03}}
 {id=WavePlayer_28, properties {soundWaveAssetPtr=&/orb_fragment_goodbye_04}}
]
properties {
 attenuationSettings=&Fort_GadgetActivate_Att
 bOverrideConcurrency=true
 maxDistance=4000
 soundClassObject=/Game/Sounds/FortSoundClassesAndModes/Fort_Gadgets_StW
 volumeMultiplier=0.35
 concurrencyOverrides {resolutionRule=PreventNew}
}
