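# Sound cue for the Orb Fragment gadget's "hello" sound in Fort Gadgets.
# A fixed one-second delay leads into an equally weighted random pick among three hello wave variations.
# It uses the gadget orb activate attenuation and the Fort gadgets sound class, with concurrency overridden on the cue itself.
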SoundCue /Game/Sounds/Fort_Gadgets/Orb_Fragment/Orb_Fragment_Hello_Cue
root=Delay_2
nodes [
 {id=Delay_2, children [Random_4], properties {delayMax=1, delayMin=1}}
 {id=Random_4, children [WavePlayer_22, WavePlayer_23, WavePlayer_24], properties {weights [1,1,1]}}
 {id=WavePlayer_22, properties {soundWaveAssetPtr=&/orb_fragment_hello_01}}
 {id=WavePlayer_23, properties {soundWaveAssetPtr=&/orb_fragment_hello_02}}
 {id=WavePlayer_24, properties {soundWaveAssetPtr=&/orb_fragment_hello_03}}
]
properties {
 attenuationSettings=/Game/Sounds/Attenuation/Fort_Gadget_Orb_Activate_Att
 bOverrideConcurrency=true
 maxDistance=4000
 soundClassObject=/Game/Sounds/FortSoundClassesAndModes/Fort_Gadgets_StW
 volumeMultiplier=1
}
