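# Sound cue for the Orb Fragment gadget that plays a "thanks" sound.
# A Random node picks one of four equally weighted wave variations and feeds it straight to the output at reduced volume.
# It uses the gadget orb-activate attenuation and the Fort gadgets sound class.
# A concurrency override prevents new instances from starting once the limit is reached, rather than stopping playing ones.
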SoundCue /Game/Sounds/Fort_Gadgets/Orb_Fragment/Orb_Fragment_Thanks_Cue
root=Random_3
nodes [
 {id=Random_3, children [WavePlayer_17, WavePlayer_18, WavePlayer_19, WavePlayer_20], properties {weights [1,1,1,1]}}
 {id=WavePlayer_17, properties {soundWaveAssetPtr=&/orb_fragment_thanks_01}}
 {id=WavePlayer_18, properties {soundWaveAssetPtr=&/orb_fragment_thanks_02}}
 {id=WavePlayer_19, properties {soundWaveAssetPtr=&/orb_fragment_thanks_03}}
 {id=WavePlayer_20, properties {soundWaveAssetPtr=&/orb_fragment_thanks_04}}
]
properties {
 attenuationSettings=/Game/Sounds/Attenuation/Fort_Gadget_Orb_Activate_Att
 bOverrideConcurrency=true
 maxDistance=4000
 soundClassObject=/Game/Sounds/FortSoundClassesAndModes/Fort_Gadgets_StW
 volumeMultiplier=0.35
 concurrencyOverrides {resolutionRule=PreventNew}
}
