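# Sound cue for an Orb Fragment gadget turning on, with attenuation and prevent-new concurrency.
# Each mixer layers a pitch-ramped opening sound, a delayed pitched-up ready sound, two delayed pings and a distant Atlas beam fading through an envelope.
# A quality-level node sends its first two tiers to one mixer and the last two to a parallel copy whose beam is a looping "opt" wave.
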SoundCue /Game/Sounds/Fort_Gadgets/Orb_Fragment/Orb_Fragment_TurnOn_Cue
root=QualityLevel_0
nodes [
 {id=Delay_0, children [WavePlayer_1], properties {delayMax=1.5, delayMin=1.5}}
 {id=Delay_1, children [Modulator_1], properties {delayMax=0.5, delayMin=0.5}}
 {id=Delay_2, children [WavePlayer_1], properties {delayMax=3.5, delayMin=3.5}}
 {id=Delay_3, children [WavePlayer_20], properties {delayMax=1.5, delayMin=1.5}}
 {id=Delay_4, children [WavePlayer_20], properties {delayMax=3.5, delayMin=3.5}}
 {id=Delay_5, children [WavePlayer_4], properties {delayMax=0.5, delayMin=0.5}}
 {id=Delay_6, children [Modulator_0], properties {delayMax=0.5, delayMin=0.5}}
 {id=Delay_7, children [WavePlayer_25], properties {delayMax=0.5, delayMin=0.5}}
 {
  id=Enveloper_0
  children [WavePlayer_18]
  properties {pitchCurve {editorCurveData {keys [<at 0.0135 0.54>,<at 1 0.9955>]}}, volumeCurve {editorCurveData {keys [<at 0 1>]}}}
 }
 {
  id=Enveloper_1
  children [Delay_7]
  properties {pitchCurve {editorCurveData {keys [{time=0}]}}, volumeCurve {editorCurveData {keys [<at 0 1>,<at 5.5973 0.8543>,<at 10 0>]}}}
 }
 {
  id=Enveloper_2
  children [WavePlayer_2]
  properties {pitchCurve {editorCurveData {keys [<at 0.0135 0.54>,<at 1 0.9955>]}}, volumeCurve {editorCurveData {keys [<at 0 1>]}}}
 }
 {
  id=Enveloper_3
  children [Delay_5]
  properties {pitchCurve {editorCurveData {keys [{time=0}]}}, volumeCurve {editorCurveData {keys [<at 0 1>,<at 5.5973 0.8543>,<at 10 0>]}}}
 }
 {id=Mixer_0, children [Delay_6,Enveloper_0,Delay_3,Delay_4,Enveloper_1], properties {inputVolume [0.6,0.5,0.5,0.3,0.3]}}
 {id=Mixer_1, children [Delay_1,Enveloper_2,Delay_0,Delay_2,Enveloper_3], properties {inputVolume [0.6,0.5,0.5,0.3,0.3]}}
 {id=Modulator_0, children [WavePlayer_21], properties {pitchMax=1.5, pitchMin=1.5}}
 {id=Modulator_1, children [WavePlayer_0], properties {pitchMax=1.5, pitchMin=1.5}}
 {id=QualityLevel_0, children [Mixer_0,Mixer_0,Mixer_1,Mixer_1]}
 {id=WavePlayer_0, properties {soundWaveAssetPtr=&/orb_fragment_ready_01}}
 {id=WavePlayer_1, properties {soundWaveAssetPtr=&/orb_fragment_ping}}
 {id=WavePlayer_18, properties {soundWaveAssetPtr=&/orb_fragment_open_part_one}}
 {id=WavePlayer_2, properties {soundWaveAssetPtr=&orb_fragment_open_part_one}}
 {id=WavePlayer_20, properties {soundWaveAssetPtr=&orb_fragment_ping}}
 {id=WavePlayer_21, properties {soundWaveAssetPtr=&orb_fragment_ready_01}}
 {id=WavePlayer_25, properties {soundWaveAssetPtr=/Game/Sounds/Fort_Atlas_Sounds/atlas_beam_distant_02}}
 {id=WavePlayer_4, properties {bLooping=true, soundWaveAssetPtr=/Game/Sounds/Fort_Atlas_Sounds/atlas_beam_distant_02_opt}}
]
properties {
 attenuationSettings=/Game/Sounds/Attenuation/Fort_Gadget_Orb_Activate_Att
 bOverrideConcurrency=true
 maxDistance=4000
 soundClassObject=/Game/Sounds/FortSoundClassesAndModes/Fort_Gadgets_StW
 volumeMultiplier=0.35
 concurrencyOverrides {resolutionRule=PreventNew}
}
