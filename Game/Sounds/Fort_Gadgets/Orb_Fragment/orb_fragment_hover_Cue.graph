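# Looping hover sound cue for the orb fragment gadget, played at reduced volume with gadget orb-activate attenuation and the gadgets sound class.
# A quality-level root selects one of two mixers: the first two levels use the standard distant Atlas beam loop, the last two its "_opt" variant.
# Each mixer layers that beam with the same looping orb fragment ping, whose start is held back by a fixed two-second delay.
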SoundCue /Game/Sounds/Fort_Gadgets/Orb_Fragment/orb_fragment_hover_Cue
root=QualityLevel_0
nodes [
 {id=Delay_0, children [WavePlayer_1], properties {delayMax=2, delayMin=2}}
 {id=Delay_4, children [WavePlayer_6], properties {delayMax=2, delayMin=2}}
 {id=Mixer_0, children [WavePlayer_24, Delay_4], properties {inputVolume [1,1]}}
 {id=Mixer_1, children [WavePlayer_2,Delay_0], properties {inputVolume [1,1]}}
 {id=QualityLevel_0, children [Mixer_0,Mixer_0,Mixer_1,Mixer_1]}
 {id=WavePlayer_1, properties {bLooping=true, soundWaveAssetPtr=&/orb_fragment_ping}}
 {id=WavePlayer_2, properties {bLooping=true, soundWaveAssetPtr=/Game/Sounds/Fort_Atlas_Sounds/atlas_beam_distant_02_opt}}
 {id=WavePlayer_24, properties {bLooping=true, soundWaveAssetPtr=/Game/Sounds/Fort_Atlas_Sounds/atlas_beam_distant_02}}
 {id=WavePlayer_6, properties {bLooping=true, soundWaveAssetPtr=&orb_fragment_ping}}
]
properties {
 attenuationSettings=/Game/Sounds/Attenuation/Fort_Gadget_Orb_Activate_Att
 bOverrideConcurrency=true
 maxDistance=4000
 soundClassObject=/Game/Sounds/FortSoundClassesAndModes/Fort_Gadgets_StW
 volumeMultiplier=0.3
}
