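# Sound cue for the second part of the orb fragment gadget's opening sound.
# A single wave player feeds the orb_fragment_open_part_two sound straight to the output at reduced volume.
# It uses the gadget orb-activate attenuation and the Fort gadgets sound class, and overrides concurrency so new instances are blocked once the limit is reached.
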SoundCue /Game/Sounds/Fort_Gadgets/Orb_Fragment/orb_fragment_open_part_two_Cue
root=WavePlayer_31
nodes [{id=WavePlayer_31, properties {soundWaveAssetPtr=&/orb_fragment_open_part_two}}]
properties {
 attenuationSettings=/Game/Sounds/Attenuation/Fort_Gadget_Orb_Activate_Att
 bOverrideConcurrency=true
 maxDistance=4000
 soundClassObject=/Game/Sounds/FortSoundClassesAndModes/Fort_Gadgets_StW
 volumeMultiplier=0.35
 concurrencyOverrides {resolutionRule=PreventNew}
}
